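# Animation Blueprint for the F_MED_80sExercise backpack cosmetic, adding physics-driven secondary motion to its body and flap bones.
# The reference pose feeds two chained Anim Dynamics nodes: a damped linear and angular spring on the body, then an angular spring with overridden gravity on the flap.
# A test sequence player, input pose and body Trail controller are left unconnected.
# The event graph has no logic and there are no variables.
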AnimBlueprint /BRCosmetics/Accessories/FORT_Backpacks/F_MED_80sExercise/Meshes/F_MED_80sExercise_F_MED_80sExercise_AnimBP
target_skeleton=/Game/Accessories/FORT_Backpacks/F_MED_80sExercise/Meshes/F_MED_80sExercise_F_MED_80sExercise_Skeleton
graphs [
 AnimGraph {nodes [
   n1 = Anim Dynamics - Bone: body(ComponentPose=Local To Component(LocalPose=Local Space Ref Pose()), Alpha=1, ExternalForce="0,0,0") {
    angularDampingOverride=0.98
    angularSpringConstant=500
    bAngularSpring=true
    bGravityOverrideInSimSpace=true
    bLinearSpring=true
    bOverrideAngularDamping=true
    bOverrideLinearDamping=true
    gravityScale=0.2
    lODThreshold=1
    linearDampingOverride=1
    linearSpringConstant=200
    alphaBoolBlend {blendOption=Linear}
    boundBone {boneName=body}
    componentAppliedLinearAccClamp=<3000 3000 3000>
    componentLinearAccScale=<0.4 0.4 0.4>
    componentLinearVelScale=<0.4 0.4 0.4>
    physicsBodyDefinitions [{
      sphereCollisionRadius=0
      boundBone {boneName=body}
      boxExtents=<30 25 15>
      constraintSetup {
       angularTargetAxis=AxisZ
       angularLimitsMax=<0 0 30>
       angularLimitsMin=<0 0 -30>
       angularTarget=<0 0 1>
       linearAxesMax=<3 1 0>
       linearAxesMin=<0 -1 0>
      }
      localJointOffset=<0 0 -2>
     }]
   }
   n2 = Anim Dynamics - Bone: flap(ComponentPose=@n1, Alpha=1, ExternalForce="0,0,0") {
    angularDampingOverride=0.95
    angularSpringConstant=3000
    bAngularSpring=true
    bGravityOverrideInSimSpace=true
    bOverrideAngularDamping=true
    bUseGravityOverride=true
    numSolverIterationsPostUpdate=4
    numSolverIterationsPreUpdate=16
    alphaBoolBlend {blendOption=Linear}
    boundBone {boneName=flap}
    gravityOverride {z=-980}
    physicsBodyDefinitions [{
      sphereCollisionRadius=0
      boundBone {boneName=flap}
      boxExtents=<5 30 15>
      constraintSetup {angularTargetAxis=AxisY, twistAxis=AxisY, angularLimitsMax=<60 0 5>, angularLimitsMin=<0 0 -5>, angularTarget=<0 1 0>}
      localJointOffset=<0 -10 0>
     }]
   }
   Output Pose(Result=Component To Local(ComponentPose=@n2)) {layerGroup=None}
   Sequence Player 'Backpack_PhysicsTest_Animation'() {sequence=/Game/Accessories/FORT_Backpacks/F_MED_80sExercise/Meshes/Backpack_PhysicsTest_Animation}
   AnimGraphNode_SubInput_0 = Input Pose()
   Trail controller - Bone: body(Alpha=0.33) {
    bReorientParentToChild=false
    bShowBaseMotion=false
    bShowLimit=false
    chainBoneAxis=Y
    lastBoneRotationAnimAlphaBlend=10
    relaxationSpeedScale=3.4616
    baseJoint {boneName=body}
    planarLimits [{drivingBone {boneName=body}, planeTransform {location=<0 0 0>, rotation=<rot 0 0 0>, scale=<1 1 1>}}]
    rotationLimits [{limitMax=<180 180 180>}, {limitMax=<180 180 180>}]
    rotationOffsets [<0 0 0>, <0 0 0>]
    trailBone {boneName=body}
    trailRelaxationSpeed {editorCurveData {keys [
       {arriveTangent=0.5175, interpMode=RCIM_Cubic, leaveTangent=0.5175, tangentMode=RCTM_User, time=0.0055, value=11.2012}
       {interpMode=RCIM_Cubic, tangentMode=RCTM_User, time=1, value=11.3548}
      ]}}
    bLimitStretch=~
   }
  ]}
 EventGraph {nodes [Event BlueprintUpdateAnimation()]}
]
variables []
